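# Activatable "Purchase" dialog for cosmetic gating in the item shop, showing gated-item purchase info with Back and hold-to-confirm Purchase buttons.
# When the offer view model updates, it caches the player's eligibility for restricted items and passes the offer to the purchase info widget.
# That eligibility picks the default focus: Purchase if eligible, otherwise Back.
# Each button fires its dispatcher (OnDismissed or OnPurchaseClicked) and deactivates the modal.
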WidgetBlueprint /FortniteCoreUI/Frontend/CosmeticGating/WBP_CosmeticGate_ShopModal
graphs [
 BP_GetDesiredFocusTarget {nodes [
   Get Desired Focus Target()
   > Return Node!(
    execute
    ReturnValue:Widget Object Reference=Select(
     Option 0:Widget Object Reference=Get Button_Back() -> (Button_Back:WBP UIKit Button Regular Object Reference)
     Option 1:Widget Object Reference=Get Button_Purchase() -> (Button_Purchase:WBP UIKit Button Regular Object Reference)
     Index=Get IsPlayerEligibleToUseRestrictedItems() -> (IsPlayerEligibleToUseRestrictedItems:bool)
    ) -> (ReturnValue:Widget Object Reference)
   )
  ]}
 EventGraph {nodes [
   "On Clicked (Button_Back)"!() -> (OutputDelegate, then, Button:Common Button Base Object Reference) {is=title K2Node_ComponentBoundEvent}
   > Call On Dismissed!(execute:Exec, self:WBP Cosmetic Gate Shop Modal Object Reference) -> (then:Exec) {is=title K2Node_CallDelegate}
   > DeactivateWidget()
   "On Hold Completed (Button_Purchase)"() {is=title K2Node_ComponentBoundEvent}
   > Call On Purchase Clicked!(execute:Exec, self:WBP Cosmetic Gate Shop Modal Object Reference) -> (then:Exec) {is=title K2Node_CallDelegate}
   > DeactivateWidget()
   Event PreConstruct()
   > n1 = Construct Object from Class(Class=/Script/FortniteUI.UIKitDialogViewModel) -> (ReturnValue:UI Kit Dialog Viewmodel Object Reference)
   > Set DialogTitle!(execute, DialogTitle:Text="LOCTEXT(\"Purchase\")", self:UI Kit Dialog Viewmodel Object Reference=@n1) -> (then, Output_Get:Text)
   > SetDialogData!(
    execute
    self:WBP UIKit Dialog Internal Object Reference=Get WBP_UIKit_Dialog_Base() -> (WBP_UIKit_Dialog_Base:WBP UIKit Dialog Internal Object Reference)
    Data:UI Kit Dialog Viewmodel Object Reference=@n1
   ) -> (then)
  ]}
 OnDismissed {nodes [OnDismissed() {is=title K2Node_FunctionEntry}]}
 OnPurchaseClicked {nodes [OnPurchaseClicked() {is=title K2Node_FunctionEntry}]}
 VM_OfferUpdated {nodes [
   n1 = VM_OfferUpdated!() -> (then, OfferVM:Fort Item Shop Offer VM Object Reference) {is=title K2Node_FunctionEntry}
   > n2 = Get Content Gating Interfaces From Offer!(execute, OfferVM:Fort Item Shop Offer VM Object Reference=@n1) -> (then, Interfaces:Array of Cosmetic Item Gating Interface Interfaces) {is=title K2Node_CallFunction}
   > n3 = IsPlayerEligibleToUseRestrictedItems!(
    execute
    LocalPlayer:Local Player Object Reference=GetOwningLocalPlayer!(self:Widget Object Reference) -> (ReturnValue:Local Player Object Reference)
    CosmeticItems:Array of Cosmetic Item Gating Interface Interfaces=@n2
   ) -> (then, ReturnValue:bool)
   > Set IsPlayerEligibleToUseRestrictedItems(IsPlayerEligibleToUseRestrictedItems:bool=@n3) -> (Output_Get:bool)
   > Set FortItemShopOfferVM!(
    execute
    FortItemShopOfferVM:Fort Item Shop Offer VM Object Reference=Get OfferVM() -> (OfferVM:Fort Item Shop Offer VM Object Reference)
    self:WBP Cosmetic Gate Purchase Info Object Reference=Get WBP_CosmeticGate_PurchaseInfo() -> (WBP_CosmeticGate_PurchaseInfo:WBP Cosmetic Gate Purchase Info Object Reference)
   ) -> (then, Output_Get:Fort Item Shop Offer VM Object Reference)
  ]}
]
variables [
 IsPlayerEligibleToUseRestrictedItems {type=boolean}
 OnDismissed {}
 OnPurchaseClicked {}
]
widgets {
 info {
  inheritedWidgetCount=0
  namedSlotCount=1
  parentClass=&CommonActivatableWidget
  rootWidgetClass=/FortUILibrary/Systems/Dialog/WBP_UIKit_Dialog_Internal.WBP_UIKit_Dialog_Internal_C
  widgetCount=7
 }
 widgets [
  WBP_UIKit_Dialog_Base {
   widgetClassPath=/FortUILibrary/Systems/Dialog/WBP_UIKit_Dialog_Internal
   bIsVariable=true
   widget=WBP_UIKit_Dialog_Base
   properties {dialogSIze {
     height Override {enable=false, value=800}
     max Aspect Ratio {enable=false, value=0}
     max Desired Height {enable=true, value=800}
     max Desired Width {enable=false, value=0}
     min Aspect Ratio {enable=false, value=0}
     min Desired Height {enable=false, value=0}
     min Desired Width {enable=true, value=864}
     width Override {enable=false, value=0}
    }}
  }
  Overlay_1 {widgetClassPath=&Overlay, namedSlotHost=WBP_UIKit_Dialog_Base, widget=Overlay_1}
  VerticalBox_0 {
   widgetClassPath=&VerticalBox
   parent=Overlay_1
   widget=VerticalBox_0
   slot {class=OverlaySlot, properties {horizontalAlignment=HAlign_Fill, padding=<ltrb 80 0 80 -64>}}
  }
  WBP_CosmeticGate_PurchaseInfo {
   widgetClassPath=/Game/Athena/UI/Frontend/CosmeticGating/WBP_CosmeticGate_PurchaseInfo
   bIsVariable=true
   parent=VerticalBox_0
   widget=WBP_CosmeticGate_PurchaseInfo
   properties {
    show Header=false
    show Header Icon=false
    show Sub Header Icon=false
    usability Matrix Alignment=HAlign_Center
    body Text Font {
     bForceMonospaced=false
     bMaterialIsStencil=false
     fontMaterial=null
     fontObject=&HeadingNow
     letterSpacing=0
     monospacedWidth=1
     size=20.25
     skewAmount=0
     typefaceFontName=74 Regular
     outlineSettings {
      bApplyOutlineToDropShadows=false
      bMiteredCorners=false
      bSeparateFillAlpha=false
      outlineMaterial=null
      outlineSize=0
      outlineColor=<rgba 0 0 0 1>
     }
    }
    header Font {
     bForceMonospaced=false
     bMaterialIsStencil=false
     fontMaterial=null
     fontObject=&HeadingNow
     letterSpacing=-10
     monospacedWidth=1
     size=25.5
     skewAmount=0
     typefaceFontName=86 Bold Italic
     outlineSettings {
      bApplyOutlineToDropShadows=false
      bMiteredCorners=false
      bSeparateFillAlpha=false
      outlineMaterial=null
      outlineSize=0
      outlineColor=<rgba 0 0 0 1>
     }
    }
    sub Header Font {
     bForceMonospaced=false
     bMaterialIsStencil=false
     fontMaterial=null
     fontObject=&HeadingNow
     letterSpacing=0
     monospacedWidth=1
     size=20.25
     skewAmount=0
     typefaceFontName=74 Regular
     outlineSettings {
      bApplyOutlineToDropShadows=false
      bMiteredCorners=false
      bSeparateFillAlpha=false
      outlineMaterial=null
      outlineSize=0
      outlineColor=<rgba 0 0 0 1>
     }
    }
   }
   slot {class=VerticalBoxSlot, properties {horizontalAlignment=HAlign_Left, verticalAlignment=VAlign_Top}}
  }
  HB_Buttons {
   widgetClassPath=&HorizontalBox
   parent=VerticalBox_0
   widget=HB_Buttons
   slot {class=VerticalBoxSlot, properties {padding=<ltrb 0 48 0 0>}}
  }
  Button_Back {
   widgetClassPath=&WBP_UIKit_Button_Regular
   bIsVariable=true
   parent=HB_Buttons
   widget=Button_Back
   properties {
    minHeight=56
    minWidth=250
    navigation=&self:WidgetTree.Button_Back.WidgetNavigation_0
    text=Back
   }
   slot {class=HorizontalBoxSlot, properties {padding=<ltrb 0 0 32 0>, size {sizeRule=Fill, value=1}}}
  }
  Button_Purchase {
   widgetClassPath=&WBP_UIKit_Button_Regular
   bIsVariable=true
   parent=HB_Buttons
   widget=Button_Purchase
   properties {
    bRequiresHold=true
    minHeight=64
    minWidth=250
    navigation=&self:WidgetTree.Button_Purchase.WidgetNavigation_0
    text=Purchase
   }
   slot {class=HorizontalBoxSlot, properties {size {sizeRule=Fill, value=1}}}
  }
 ]
}
